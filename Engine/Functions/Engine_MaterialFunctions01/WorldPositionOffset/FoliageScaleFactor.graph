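# Material function, exposed to the library, that gets object scale for foliage meshes where simpler options like ObjectScale do not work.
# It transforms the local origin and a unit Z point into world space and takes their distance as the scale factor, so only Z-axis scale is measured.
# An optional Initial Radius input, defaulting to 1, multiplies that factor to give a scaled radius.
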MaterialFunction /Engine/Functions/Engine_MaterialFunctions01/WorldPositionOffset/FoliageScaleFactor
graph {
 properties {bExposeToLibrary=true, description="A way to get object scale for Foliage meshes where easier solutions like \"ObjectScale\" do not work."}
 nodes [
  n1 = TransformPosition(None=Constant3Vector())
  n2 = TransformPosition(None=Constant3Vector() {constant=<rgba 0 0 1 0>})
  n3 = FunctionInput() {
   bUsePreviewValueAsDefault=true
   description=This method allows you to input the object initial radius and it will multiply by the Scale Factor. Otherwise it will output the Scale Factor of the mesh.
   inputName=Initial Radius
   inputType=FunctionInput_Scalar
   previewValue=<1 0 0 1>
  }
  FunctionOutput(None=Multiply(A=Distance(A=@n1, B=@n2), B=@n3) {constB=628})
 ]
}
